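# Translucent, lit material for the Holoheart vehicle wheel cosmetic that draws glowing, scrolling hologram-style bands.
# A tiled V texture coordinate pans over time, and its fractional part times its inverse forms soft repeating parabolic bands without any texture.
# The biased and scaled bands, tinted by the ColorFX parameter (light blue by default), drive emissive.
# Opacity comes from the vertex color red channel, scaled down for a faint see-through look.
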Material /VehicleCosmetics/Wheels/Holoheart/Materials/M_Wheel_Holoheart
graph {
 properties {
  bAutomaticallySetUsageInEditor=false
  bUseMaterialAttributes=true
  bUsedWithMeshDeformer=true
  bUsedWithSkeletalMesh=true
  blendMode=BLEND_Translucent
  pixelDepthOffsetMode=PDOM_Legacy
  usedShadingModels=Default Lit
 }
 nodes [
  n1 = ComponentMask(None=TextureCoordinate() {vTiling=24}) {g=true}
  n2 = Multiply(B=Time()) {constA=2}
  n3 = Frac(None=ComponentMask(None=Panner(Coordinate=@n1, Time=@n2) {speedY=-1}) {g=true})
  n4 = ConstantBiasScale(None=Multiply(A=@n3, B=OneMinus(None=@n3))) {bias=0.1, scale=2}
  MakeMaterialAttributes!(
   BaseColor
   Metallic
   Specular
   Roughness
   Anisotropy
   EmissiveColor=Multiply(A=Multiply(A=@n4), B=VectorParameter() {group=0_Special Effects, parameterName=ColorFX, defaultValue=<rgba 0.1 0.4 1 1>}.RGB)
   Opacity=Multiply(A=VertexColor().R) {constB=0.25}
   OpacityMask
   Normal
   Tangent
   WorldPositionOffset
   SubsurfaceColor
   ClearCoat
   ClearCoatRoughness
   AmbientOcclusion
   Refraction
   CustomizedUVs_0
   CustomizedUVs_1
   CustomizedUVs_2
   CustomizedUVs_3
   CustomizedUVs_4
   CustomizedUVs_5
   CustomizedUVs_6
   CustomizedUVs_7
   PixelDepthOffset
   ShadingModel
   Displacement
  )
 ]
}
